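# Translucent unlit Niagara mesh particle material that fakes a point light on the OliveStomp glider.
# A local light position yields distance falloff blended with N·L, using the normal map, or vertex normals at Low and Medium quality.
# Softened thresholds cut this into core, outer and outline bands, with colors, radii and opacities from dynamic parameters.
# The per-band opacity blend reads Opacity Outline twice, leaving Opacity Outer unused.
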Material /BRCosmetics/Gadgets/Assets/VinderTech_GliderChute/Glider_OliveStomp/FX/Materials/M_Glider_OliveStomp_FakeLight_Mesh
graph {
 properties {
  bAutomaticallySetUsageInEditor=false
  bUsedWithNiagaraMeshParticles=true
  blendMode=BLEND_Translucent
  pixelDepthOffsetMode=PDOM_Legacy
  shadingModel=MSM_Unlit
  usedShadingModels=Unlit
 }
 nodes [
  MaterialFunctionCall(Emissive=NamedRerouteUsage()) {materialFunction=&MF_FX_EmissiveAdjustments}
  Multiply(A=NamedRerouteUsage(), B=NamedRerouteUsage())
  n1 = CameraPositionWS()
  n2 = WorldPosition()
  Multiply(A=Normalize(VectorInput=Subtract(A=@n1, B=@n2.XYZ)), B=ScalarParameter() {defaultValue=1, parameterName=Camera Offset})
  n3 = TextureSample() {samplerType=SAMPLERTYPE_Normal, texture=/BRCosmetics/Gadgets/Assets/VinderTech_GliderChute/Glider_OliveStomp/Textures/T_OliveStomp_Glider_N}
  n4 = VertexNormalWS()
  n5 = Subtract(
   A=TransformPosition(None=VectorParameter() {desc="Max Radius = 33", parameterName=Light Position Local, defaultValue=<rgba 0 0 33 33>}.RGB) {transformType=TRANSFORMPOSSOURCE_World}
   B=WorldPosition().XYZ
  )
  n6 = DotProduct(A=QualitySwitch(Default=Transform(None=@n3.RGB), Low=@n4, Medium=@n4), B=Normalize(VectorInput=@n5))
  n7 = Length(None=@n5) {is=class MaterialExpressionLength}
  n8 = DynamicParameter() {paramNames [Light Radius, Opacity Core, Opacity Outer, Opacity Outline]}
  n9 = OneMinus(None=Saturate(None=Divide(A=@n7, B=@n8.Light Radius)))
  n10 = Multiply(A=LinearInterpolate(A=@n6, B=ConstantBiasScale(None=@n6), Alpha=SquareRoot(None=@n9)), B=@n9)
  n11 = DynamicParameter() {parameterIndex=1, paramNames [Color Core R, Color Core G, Color Core B, Color Core Radius]}
  n12 = OneMinus(None=@n11.Color Core Radius)
  n13 = ScalarParameter() {defaultValue=0.01, parameterName=Fake Light Softness}
  n14 = Saturate(None=Divide(A=Subtract(A=@n10, B=@n12), B=@n13))
  n15 = DynamicParameter() {parameterIndex=2, paramNames [Color Outer R, Color Outer G, Color Outer B, Color Outline Radius]}
  n16 = OneMinus(None=@n15.Color Outline Radius)
  n17 = Saturate(None=Divide(A=Subtract(A=@n10, B=@n16), B=@n13))
  n18 = DynamicParameter() {parameterIndex=3, paramNames [Color Outline R, Color Outline G, Color Outline B, Color Outline Radius]}
  n19 = Divide(A=Subtract(A=@n10, B=OneMinus(None=@n18.Color Outline Radius)), B=@n13)
  NamedRerouteDeclaration(None=Max(A=Max(A=@n14, B=@n17), B=Saturate(None=@n19))) {name=Opacity, nodeColor=<rgba 0.7813 0 1 1>}
  NamedRerouteDeclaration(None=LinearInterpolate(A=LinearInterpolate(A=@n18.RGB, B=@n15.RGB, Alpha=@n17), B=@n11.RGB, Alpha=@n14)) {name=Emissive, nodeColor=<rgba 0 1 0.4609 1>}
  n20 = DynamicParameter() {paramNames [Light Radius, Opacity Core, Opacity Outer, Opacity Outline]}
  NamedRerouteDeclaration(None=LinearInterpolate(A=LinearInterpolate(A=@n20.Opacity Outline, B=@n20.Opacity Outline, Alpha=@n17), B=@n20.Opacity Core, Alpha=@n14)) {name=Opacity Per Band, nodeColor=<rgba 0.4531 0 1 1>}
 ]
}
