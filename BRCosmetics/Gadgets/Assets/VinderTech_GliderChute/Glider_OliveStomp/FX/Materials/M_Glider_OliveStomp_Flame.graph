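# Unlit translucent Niagara sprite material that draws a procedural, hard-edged flame for the OliveStomp glider effects.
# Panning high-contrast noise, time-offset per particle, distorts horizontal UVs feeding a tapered shape built from two dot products.
# Particle alpha shifts the shape's threshold, then a square-root fade at one vertical end and rounding produce a stepped mask.
# Particle color feeds emissive through a shared FX adjustment function.
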Material /BRCosmetics/Gadgets/Assets/VinderTech_GliderChute/Glider_OliveStomp/FX/Materials/M_Glider_OliveStomp_Flame
graph {
 properties {
  bAutomaticallySetUsageInEditor=false
  bUsedWithNiagaraSprites=true
  blendMode=BLEND_Translucent
  pixelDepthOffsetMode=PDOM_Legacy
  shadingModel=MSM_Unlit
  usedShadingModels=Unlit
 }
 nodes [
  n1 = ParticleColor()
  MaterialFunctionCall(Emissive=@n1.RGB) {materialFunction=&MF_FX_EmissiveAdjustments}
  n2 = MaterialFunctionCall(Float2=TextureCoordinate()) {materialFunction=&BreakOutFloat2Components}
  n3 = TextureCoordinate()
  n4 = VectorParameter() {parameterName=UV Noise Scale and Seped, defaultValue=<rgba 0.2 0.6 0.2 1.11>}
  n5 = ComponentMask(None=@n4.RGB) {g=true, r=true}
  n6 = Time()
  n7 = Multiply(A=ParticleRandom()) {constB=6.28}
  n8 = TextureSample(UVs=Panner(Coordinate=Multiply(A=@n3, B=@n5), Time=Add(A=@n6, B=@n7), Speed=AppendVector(A=@n4.B, B=@n4.B)) {bFractionalPart=true, speedY=1}) {samplerType=SAMPLERTYPE_LinearColor, texture=&HighContrastNoise}
  n9 = Add(A=@n2.R, B=Multiply(A=Subtract(A=@n8.R) {constB=0.5}, B=ScalarParameter() {defaultValue=0.166, parameterName=UV Distortion Intensity}))
  n10 = OneMinus(None=@n2.G)
  n11 = Add(
   A=ConstantBiasScale(None=MaterialFunctionCall(X=@n9, Y=@n10) {materialFunction=&MakeFloat2}.Result) {bias=-0.5, scale=2}
   B=Constant2Vector() {g=0.166}
  )
  n12 = DotProduct(A=@n11, B=Constant2Vector() {g=0.333, r=-1})
  n13 = DotProduct(A=@n11, B=Constant2Vector() {g=0.333, r=1})
  n14 = Add(A=Subtract(A=Max(A=@n12, B=@n13), B=@n1.A))
  n15 = ComponentMask(None=TextureCoordinate()) {g=true}
  n16 = Saturate(None=Divide(A=OneMinus(None=@n15)) {constB=0.222})
  Round(None=Multiply(A=OneMinus(None=@n14), B=SquareRoot(None=@n16)))
 ]
}
