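# Unlit translucent smoke material for the OliveStomp glider FX, used on Niagara mesh particles with spline-thicken geometry.
# World position offset widens strips toward the camera, with dynamic parameters driving bend, UV offset, fade and an Android mobile-path tangent fix.
# Squished, distorted, panning UVs sample an SDF smoke texture through MF_DistanceField for fill and noise-varied outline.
# A second, light-offset SDF sample is subtracted to fake directional lighting, blending base and outline colors toward lit tints.
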Material /BRCosmetics/Gadgets/Assets/VinderTech_GliderChute/Glider_OliveStomp/FX/Materials/M_Glider_OliveStomp_Smoke_SplineThicken
graph {
 properties {
  bAutomaticallySetUsageInEditor=false
  bUsedWithNiagaraMeshParticles=true
  blendMode=BLEND_Translucent
  pixelDepthOffsetMode=PDOM_Legacy
  shadingModel=MSM_Unlit
  usedShadingModels=Unlit
 }
 nodes [
  n1 = ParticleColor()
  n2 = NamedRerouteUsage()
  n3 = VectorParameter() {parameterName=Lit Color, defaultValue=<rgba 0.7 0.2695 0.07 0>}
  n4 = NamedRerouteUsage()
  n5 = TextureSample(UVs=@n4) {
   samplerType=SAMPLERTYPE_LinearGrayscale
   texture=/BRCosmetics/Gadgets/Assets/VinderTech_GliderChute/Glider_OliveStomp/FX/Textures/T_Glider_OliveStomp_Smoke_SDF
  }
  n6 = NamedRerouteUsage()
  n7 = MaterialFunctionCall!(RGBA Channels=Multiply(A=Saturate(None=@n5.A), B=NamedRerouteUsage()), Glow Min, Glow Max, Outline Thickness=@n6) {materialFunction=&/MF_DistanceField}
  n8 = NamedRerouteUsage()
  n9 = ScalarParameter() {defaultValue=4, parameterName=Fake Lighting Offset Exponent}
  n10 = Multiply(A=Multiply(A=Power(Base=@n8, Exp=@n9), B=NamedRerouteUsage()), B=ScalarParameter() {defaultValue=-1, parameterName=Fake Lighting UV Offset})
  n11 = Constant()
  n12 = MaterialFunctionCall(Float2=Add(A=@n4, B=AppendVector(A=@n10, B=@n11))) {materialFunction=&BreakOutFloat2Components}
  n13 = TextureSample(UVs=MaterialFunctionCall(X=Saturate(None=@n12.R), Y=@n12.G) {materialFunction=&MakeFloat2}.Result) {
   samplerType=SAMPLERTYPE_LinearGrayscale
   texture=&T_Glider_OliveStomp_Smoke_SDF
  }
  n14 = Saturate(None=@n13.A)
  n15 = NamedRerouteUsage()
  n16 = Subtract(
   A=@n7.Fill
   B=MaterialFunctionCall(RGBA Channels=Multiply(A=@n14, B=@n15), Outline Thickness=@n6) {materialFunction=&MF_DistanceField}.Fill
  )
  n17 = Multiply(A=NamedRerouteUsage(), B=Multiply(A=NamedRerouteUsage(), B=Saturate(None=@n16)))
  n18 = LinearInterpolate(
   A=LinearInterpolate(A=VectorParameter() {parameterName=Color, defaultValue=<rgba 0.1 0.07 0.06 0>}.RGB, B=@n3.RGB, Alpha=NamedRerouteUsage())
   B=@n3.RGB
   Alpha=@n17
  )
  n19 = VectorParameter() {parameterName=Outline Lit Color, defaultValue=<rgba 0.6 0.2059 0.0304 1>}
  n20 = LinearInterpolate(
   A=LinearInterpolate(A=VectorParameter() {parameterName=Outline Color, defaultValue=<rgba 0.05 0.0365 0.032 0>}.RGB, B=@n19.RGB, Alpha=NamedRerouteUsage())
   B=@n19.RGB
   Alpha=@n17
  )
  MaterialFunctionCall(Emissive=Multiply(A=LinearInterpolate(A=@n1.RGB, Alpha=@n2), B=LinearInterpolate(A=@n18, B=@n20, Alpha=@n7.Outline))) {materialFunction=&MF_FX_EmissiveAdjustments}
  n21 = TextureCoordinate()
  n22 = MaterialFunctionCall(Float2=@n21) {materialFunction=&BreakOutFloat2Components}
  n23 = Subtract(A=@n22.R) {constB=0.5}
  n24 = ComponentMask(None=@n21) {g=true}
  n25 = ScalarParameter() {defaultValue=3, parameterName=UV Squish Exponent}
  n26 = NamedRerouteDeclaration(None=OneMinus(None=Power(Base=@n24, Exp=@n25))) {name=Bottom Fade, nodeColor=<rgba 0 0.6719 1 1>}
  n27 = MaterialFunctionCall(X=Add(A=Divide(A=@n23, B=@n26)) {constB=0.5}, Y=@n22.G) {materialFunction=&MakeFloat2}
  n28 = Constant()
  n29 = DynamicParameter() {paramNames [UV Offset, WPO X, WPO Y, Opacity]}
  n30 = Multiply(
   A=Add(A=Saturate(None=@n27.Result), B=AppendVector(A=@n28, B=@n29.UV Offset))
   B=AppendVector(A=Constant() {r=1}, B=ScalarParameter() {defaultValue=1.333, parameterName=UV Scale})
  )
  n31 = NamedRerouteUsage()
  n32 = Multiply(
   A=Subtract(A=ComponentMask(None=@n31) {g=true, r=true}) {constB=0.5}
   B=Multiply(A=ScalarParameter() {defaultValue=0.333, parameterName=UV Distortion Intensity}, B=@n26)
  )
  NamedRerouteDeclaration(None=Panner(Coordinate=Add(A=@n30, B=@n32)) {speedY=1}) {name=Panning UVs, nodeColor=<rgba 1 0.9375 0 1>}
  n33 = Transform(None=Constant3Vector() {constant=<rgba 0 1 0 0>})
  n34 = DynamicParameter() {
   desc="This only exists to fix the splinethicken meshes on android,\r\napparently the tangents do not properly respond to negative scale there"
   parameterIndex=1
   defaultValue=<rgba 1 1 1 0>
   paramNames [Scale X, Scale Y, Scale Z, ""]
  }
  n35 = ShadingPathSwitch(Default=@n33, Deferred=@n33, Mobile=Multiply(A=@n33, B=@n34.RGB)) {desc=Migrated from FeatureLevelSwitch}
  n36 = ComponentMask(None=TextureCoordinate()) {g=true}
  n37 = DynamicParameter() {paramNames [UV Offset, WPO X, WPO Y, Opacity]}
  n38 = Multiply(A=Power(Base=OneMinus(None=@n36)), B=AppendVector(A=AppendVector(A=@n37.WPO X, B=@n37.WPO X), B=Constant()))
  n39 = Subtract(A=CameraPositionWS(), B=Add(A=WorldPosition().XYZ, B=@n38))
  n40 = Normalize(VectorInput=CrossProduct(A=@n35, B=Normalize(VectorInput=@n39)))
  n41 = CrossProduct(A=@n40, B=Normalize(VectorInput=@n35))
  NamedRerouteDeclaration(None=VertexInterpolator(VS=Normalize(VectorInput=@n41)).PS) {name=SplineThicken Normal, nodeColor=<rgba 0 0.3437 1 1>}
  n42 = VectorParameter() {parameterName=UV Noise Scale and Seped, defaultValue=<rgba 0.2 1 0.3 0.6>}
  n43 = Multiply(A=TextureCoordinate(), B=ComponentMask(None=@n42.RGB) {g=true, r=true})
  n44 = AppendVector(A=@n42.B, B=@n42.B)
  NamedRerouteDeclaration(None=TextureSample(UVs=Panner(Coordinate=@n43, Speed=@n44) {bFractionalPart=true, speedY=1}) {samplerType=SAMPLERTYPE_LinearColor, texture=&HighContrastNoise}.RGB) {name=VariationNoise, nodeColor=<rgba 0 0.3906 1 1>}
  n45 = OneMinus(None=ComponentMask(None=TextureCoordinate()) {g=true})
  n46 = Multiply(
   A=DynamicParameter() {paramNames [UV Offset, WPO X, WPO Y, Opacity]}.Opacity
   B=ScalarParameter() {defaultValue=3, parameterName=Top Fade Exponent}
  )
  NamedRerouteDeclaration(None=OneMinus(None=Power(Base=@n45, Exp=@n46))) {name=Top Fade, nodeColor=<rgba 0.7109 1 0 1>}
  n47 = ScalarParameter() {parameterName=Top Outline Thickness}
  n48 = LinearInterpolate(
   A=ScalarParameter() {parameterName=Outline Tikcness Min}
   B=ScalarParameter() {defaultValue=0.4, parameterName=Outline Thickness Max}
   Alpha=ComponentMask(None=NamedRerouteUsage()) {r=true}
  )
  n49 = NamedRerouteUsage()
  NamedRerouteDeclaration(None=Divide(A=LinearInterpolate(A=@n47, B=@n48, Alpha=@n49), B=NamedRerouteUsage())) {name=Outline Thickness, nodeColor=<rgba 0.8281 0 1 1>}
  n50 = TransformPosition(None=ParticlePositionWS()) {transformSourceType=TRANSFORMPOSSOURCE_World}
  n51 = VectorParameter() {parameterName=Light Local Position, defaultValue=<rgba 0 21 0 0>}
  n52 = Normalize(VectorInput=Multiply(A=Subtract(A=@n50, B=@n51.RGB), B=Constant3Vector() {constant=<rgba 1 1 0 0>}))
  n53 = Transform(None=@n52) {transformSourceType=TRANSFORMSOURCE_Local, transformType=TRANSFORM_View}
  NamedRerouteDeclaration(None=Sign(None=ComponentMask(None=@n53) {r=true})) {name=Fake Lighting Sign, nodeColor=<rgba 0.7109 0 1 1>}
  n54 = DotProduct(A=NamedRerouteUsage(), B=Transform(None=@n52) {transformSourceType=TRANSFORMSOURCE_Local})
  n55 = NamedRerouteDeclaration(None=OneMinus(None=ConstantBiasScale(None=@n54))) {name=Fake Lighting, nodeColor=<rgba 1 0 0.6094 1>}
  n56 = ScalarParameter() {defaultValue=9, parameterName=Fake Lighting Backlighting Exponent}
  NamedRerouteDeclaration(None=Multiply(A=Power(Base=@n55, Exp=@n56), B=NamedRerouteUsage())) {name=Backlighting Alpha, nodeColor=<rgba 1 0 0.0937 1>}
  NamedRerouteDeclaration(None=Power(Base=@n55, Exp=ScalarParameter() {defaultValue=1, parameterName=Fake Lighting Exponent})) {name=Front Lighting Alpha, nodeColor=<rgba 1 0.5391 0 1>}
  n57 = TextureCoordinate()
  NamedRerouteDeclaration(None=Power(Base=ComponentMask(None=@n57) {g=true}) {constExponent=3}) {name=Front Lighting Top Fade, nodeColor=<rgba 0.125 0 1 1>}
  n58 = TextureCoordinate()
  n59 = Multiply(A=Subtract(A=ComponentMask(None=@n58) {r=true}) {constB=0.5}, B=ScalarParameter() {defaultValue=50, parameterName=Width})
  NamedRerouteDeclaration(None=Add(A=@n38, B=Multiply(A=@n40, B=@n59))) {name=SplineThicken WPO, nodeColor=<rgba 0.2891 0 1 1>}
  NamedRerouteUsage()
  Saturate(None=@n7.Fill)
 ]
}
