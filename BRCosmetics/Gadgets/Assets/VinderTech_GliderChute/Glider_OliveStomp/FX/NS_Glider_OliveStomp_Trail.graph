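# Niagara trail effect for the OliveStomp glider: flickering sprite flames, a fake light with soft glow, four smoke meshes and wind meshes.
# System-level timeline, deploy fade-in and time-of-day values drive intensity, size and tint across emitters.
# Smoke tilts against horizontal owner velocity and receives interpolated WPO material parameters; wind aligns with velocity and spawns based on a movement-input threshold.
# Front-end modifiers substitute fixed orientations and smoke color outside gameplay.
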NiagaraSystem /BRCosmetics/Gadgets/Assets/VinderTech_GliderChute/Glider_OliveStomp/FX/NS_Glider_OliveStomp_Trail
dependencies {
 usedDataInterfaces [
  &NiagaraDataInterfaceArrayFloat
  &NiagaraDataInterfaceCurve
  &NiagaraDataInterfaceVector2DCurve
  &NiagaraDataInterfaceColorCurve
 ]
 usedDynamicInputs [
  &MakeFloatFromVector
  &ClampFloat
  &Divide_Float
  &Subtract_Float
  &Subtract_Position
  /Niagara/DynamicInputs/TypeConversions/Vector2DLength
  &FNTimeOfDayScalar
  &ColorFromCurve
 ]
 usedModules [
  &SystemState
  &InterpolateOverTime
  &self:SystemSpawnScript.SystemScriptSource.SystemScriptGraph.NiagaraNodeAssignment_1.SetVariables_#1
  &self:SystemSpawnScript.SystemScriptSource.SystemScriptGraph.NiagaraNodeAssignment_2.SetVariables_#2
  &Timeline
 ]
}
emitters {
 Flames {
  emitterSpawnScript [/SetVariables(
    Emitter.FlameCoreColor:LinearColor=<rgba 6 2.2 0 0.55>
    Emitter.FlameOuterColor:LinearColor=<rgba 3 0.65 0 0.75>
    Emitter.FlameOutlineColor:LinearColor=<rgba 0.682 0.0682 0 0.85>
   ) {isSetParameters=true}]
  emitterUpdateScript [
   /SetVariables(Emitter.IncrementalTime:NiagaraFloat=&Add_Float(
     A=$Emitter.IncrementalTime
     B=&Multiply_Float(A=$Engine.DeltaTime, B=&Lerp_Float(A=0.5, B=1.5, Alpha=$System.Timeline.Time))
    )) {isSetParameters=true}
   &EmitterState(
    Inactive Response=::NewEnumerator1
    Life Cycle Mode=::NewEnumerator1
    Loop Behavior=::NewEnumerator1
    Loop Duration Mode=::NewEnumerator1
    UseLoopDelay=false
   )
   &SpawnBurst_Instantaneous(Spawn Count=6, Use Spawn Probability=false, UseLoopCountLimit=false)
  ]
  particleSpawnScript [&InitializeParticle(
    Color Mode=::NewEnumerator0
    Position Offset=<0 0 22>
    Sprite Rotation Angle=&SelectFloatFromArray(
     Direct Array Index=&ReturnParticleIDAttributes()
     Direct Array Mode=::NewEnumerator1
     Float Selection Array=DataInterface() {propertyValues {FloatData [0,-0.03,0.026,-0.06,0.075,-0.08]}}
    )
    Sprite Rotation Mode=::NewEnumerator1
    Sprite Size=&Multiply_Vector2D_ByFloat(
     Float=&SelectFloatFromArray(
      Direct Array Index=&ReturnParticleIDAttributes()
      Direct Array Mode=::NewEnumerator1
      Float Selection Array=DataInterface() {propertyValues {FloatData [1,0.6,0.8,0.7,0.6,0.6]}}
     )
     Vector2D=<31 40>
    )
    Sprite Size Mode=::NewEnumerator4
    UsePositionOffset=true
    Write Material Random=true
   )]
  particleUpdateScript [
   &ParticleState(Kill Particles When Lifetime Has Elapsed=false)
   &ScaleSpriteSize(
    Scale Factor=&Multiply_Vector2D_ByFloat(
     Float=&Waveform(
      Global Phase=&Add_Float(A=&RandomRangeFloat(Minimum=0, Maximum=1), B=$Emitter.IncrementalTime)
      "Waveform [1]"=::NewEnumerator12
      "[1] Phase Scale"=9
      "[1] Frequency"=1
      "[1] Amplitude Min/Max"=&MakeVector2D(
       X=&SelectFloatFromArray(
        Direct Array Index=&ReturnParticleIDAttributes()
        Direct Array Mode=::NewEnumerator1
        Float Selection Array=DataInterface() {propertyValues {FloatData [0.6,0.75,0.7,0.6,0.5,0.5]}}
       )
       Y=1
      )
      "[1] Scale"=&Lerp_Float(
       A=&SelectFloatFromArray(
        Direct Array Index=&ReturnParticleIDAttributes()
        Direct Array Mode=::NewEnumerator1
        Float Selection Array=DataInterface() {propertyValues {FloatData [0.5,0.6,0.6,0,0,0]}}
       )
       B=1
       Alpha=$System.Timeline.Time
      )
     )
     Vector2D=&Multiply_Vector2D_ByFloat(Float=$System.DeployFadeIn, Vector2D=<1 1>)
    )
    Scale Sprite Size Mode=::NewEnumerator2
   )
  ]
  rendererClasses [&NiagaraSpriteRendererProperties]
  renderers [
   {rendererClass=&NiagaraSpriteRendererProperties}
   {rendererClass=&NiagaraSpriteRendererProperties, rendererIndex=1}
   {rendererClass=&NiagaraSpriteRendererProperties, rendererIndex=2}
  ]
 }
 LightAndGlow {
  emitterSpawnScript [/SetVariables(
    Emitter.FakeLightCoreColor:LinearColor=<rgba 5 1.8333 0 0.8>
    Emitter.FakeLightOuterColor:LinearColor=<rgba 1.333 0.2888 0 0.93>
    Emitter.FakeLightOutlineColor:LinearColor=<rgba 0.4 0.058 0.04 1>
    Emitter.SoftGlowCameraOffset=22
    Emitter.SoftGlowColor:LinearColor=<rgba 1 0.15 0 1>
    Emitter.SoftGlowPosition:NiagaraPosition=<0 0 25>
    Emitter.SoftGlowSize:Vector2f=&Vector2DFromFloat(Value=111)
   ) {isSetParameters=true}]
  emitterUpdateScript [
   &EmitterState(
    Inactive Response=::NewEnumerator1
    Life Cycle Mode=::NewEnumerator1
    Loop Behavior=::NewEnumerator1
    Loop Duration Mode=::NewEnumerator1
    UseLoopDelay=false
   )
   &SpawnBurst_Instantaneous(Use Spawn Probability=false, UseLoopCountLimit=false)
   /SetVariables(
    Emitter.SoftGlowColor:LinearColor=&MultiplyLinearColor(
     Linear Color=$Emitter.Initial.SoftGlowColor
     Scale Factor=&MakeFloat4FromVectorAndFloat(
      "Vector (RGB)"=&VectorFromFloat(Value=&Multiply_Float(A=$System.DeployFadeIn, B=$System.TODBrightnessMultiplier))
      "Float (Alpha)"=&Lerp_Float(A=0.666, B=1, Alpha=&Power(a=$System.Timeline.Time, b=3))
     )
    )
    Emitter.SoftGlowSize:Vector2f=&Multiply_Vector2D_ByFloat(
     Float=&Lerp_Float(A=0.333, B=1, Alpha=&Power(a=$System.Timeline.Time, b=3))
     Vector2D=$Emitter.Initial.SoftGlowSize
    )
   ) {isSetParameters=true}
  ]
  particleSpawnScript [/SetVariables(
    Particles.DynamicMaterialParameter1:Vector4f=&MakeVector4_FromLinearColor(LinearColor=$Emitter.FakeLightCoreColor)
    Particles.DynamicMaterialParameter2:Vector4f=&MakeVector4_FromLinearColor(LinearColor=$Emitter.FakeLightOuterColor)
    Particles.DynamicMaterialParameter3:Vector4f=&MakeVector4_FromLinearColor(LinearColor=$Emitter.FakeLightOutlineColor)
   ) {isSetParameters=true}]
  particleUpdateScript [/SetVariables(Particles.DynamicMaterialParameter:Vector4f=&MakeVector4(
     X=&Waveform(
      "Waveform [1]"=::NewEnumerator12
      "[1] Phase Scale"=11
      "[1] Frequency"=1
      "[1] Amplitude Min/Max"=&Lerp_Vector2(A=<13 15>, B=<29 38>, Alpha=$System.Timeline.Time)
      "[1] Scale"=$System.DeployFadeIn
     )
     Y=1
     Z=1
     W=1
    )) {isSetParameters=true}]
  rendererClasses [&NiagaraMeshRendererProperties, &NiagaraSpriteRendererProperties]
  renderers [
   {rendererClass=&NiagaraMeshRendererProperties, rendererIndex=0}
   {rendererClass=&NiagaraSpriteRendererProperties, rendererIndex=1}
  ]
 }
 Smokes {
  emitterUpdateScript [
   &EmitterState(
    Inactive Response=::NewEnumerator1
    Life Cycle Mode=::NewEnumerator1
    Loop Behavior=::NewEnumerator1
    Loop Duration Mode=::NewEnumerator1
    UseLoopDelay=false
   )
   &SpawnBurst_Instantaneous(Spawn Count=4, Use Spawn Probability=false, UseLoopCountLimit=false)
  ]
  particleSpawnScript [
   &PartitionParticles(
    Normalized Index In Partition Includes One=true
    Number Of Partitions=4
    Particles Per Partition:NiagaraInt32=$Emitter.ParticlesPerRibbon
    Partition Mode:ENiagara_PartitionMode=::NewEnumerator1
    Write Partition to Ribbon ID=::NewEnumerator0
   )
   &InitializeParticle(
    Color Mode=::NewEnumerator0
    Position Offset=&MakeVector(
     X=&SelectFloatFromArray(
      Direct Array Index=$Particles.PartitionParticles.Partition
      Direct Array Mode=::NewEnumerator1
      Float Selection Array=DataInterface() {propertyValues {FloatData [31,0,-31,0]}}
     )
     Y=&SelectFloatFromArray(
      Direct Array Index=$Particles.PartitionParticles.Partition
      Direct Array Mode=::NewEnumerator1
      Float Selection Array=DataInterface() {propertyValues {FloatData [0,31,0,-31]}}
     )
     Z=50
    )
    UsePositionOffset=true
    Write Material Random=true
   )
  ]
  particleUpdateScript [
   &DynamicMaterialParameters(
    Index 0 Param 1=&SelectFloatFromArray(
     Direct Array Index=$Particles.PartitionParticles.Partition
     Direct Array Mode=::NewEnumerator1
     Float Selection Array=DataInterface() {propertyValues {FloatData [0,0.618,1,2.618]}}
    )
    Index 0 Param 2=&MakeFloatFromVector2d(Vector 2D=$System.InterpolateWPO)
    Index 0 Param 3=&MakeFloatFromVector2d(Vector 2D=$System.InterpolateWPO, Channel=::NewEnumerator1)
    Index 0 Param 4=/Niagara/DynamicInputs/Math/Min_Float(A:NiagaraFloat=$System.DeployFadeIn, B:NiagaraFloat=$System.Timeline.Time)
    Module.Index0Param0Write=true
    Module.Index0Param1Write=true
    Module.Index0Param2Write=true
    Module.Index0Param3Write=true
   )
   &ParticleState(Kill Particles When Lifetime Has Elapsed=false)
   &UpdateMeshOrientation(
    Facing Interpolation=false
    Orientation Method=::NewEnumerator1
    Override Quat=false
    TransformFacing=false
    TransformUp=true
    Up Coordinate Space:ENiagaraCoordinateSpace=::World
    Up Direction:Vector3f=&NormalizeVector(
     Vector To Normalize=&FN_FrontEndVector3Modifier(
      Gameplay=&Lerp_Vector(
       A=<0 0 1>
       B=&NormalizeVector(
        Vector To Normalize=&Multiply_Vector(A=$Engine.Owner.Velocity, B=<-1 -1 0>)
        Safe Normalize=true
        Safe Fallback Threshold=0
        Safe Fallback Vector=<0 0 0.001>
       )
       Alpha=&Multiply_Float(A=$System.Timeline.Time, B=0.25)
      )
      Front End=<0 0 1>
     )
     Safe Normalize=true
     Safe Fallback Threshold=0
     Safe Fallback Vector=<0 0 0.001>
    )
   )
   /SetVariables(Particles.Color=&MakeLinearColorFromVectorAndFloat(
     "Vector (RGB)"=&FN_FrontEndVector3Modifier(Gameplay=&MakeVectorFromLinearColorRGB(Color=$System.TODColorTint), Front End=<0.5 0.5 0.5>)
     "Float (Alpha)"=1
    )) {isSetParameters=true}
  ]
  rendererClasses [&NiagaraMeshRendererProperties]
  renderers [{rendererClass=&NiagaraMeshRendererProperties, rendererIndex=0}]
 }
 Wind {
  emitterUpdateScript [
   &EmitterState(
    Inactive Response=::NewEnumerator1
    Life Cycle Mode=::NewEnumerator1
    Loop Behavior=::NewEnumerator1
    Loop Duration Mode=::NewEnumerator1
    UseLoopDelay=false
   )
   &SpawnRate(
    Spawn Probability=&MakeCustomFloatFromBool(Bool=&SetBoolByFloatComparison(A=$System.MovementInput, B=0.05), True Float=1, False Float=0)
    SpawnRate=&FN_FrontEndFloatModifier(Gameplay=4, Front End=4)
    Use Spawn Probability=true
   )
   /SetVariables(Emitter.OrientationVector:Vector3f=&NormalizeVector(
     Vector To Normalize=&FN_FrontEndVector3Modifier(Gameplay=&Multiply_Vector(A=$Engine.Owner.Velocity, B=<1 1 0>), Front End=<1 1 -0.25>)
     Safe Normalize=true
     Safe Fallback Threshold=0
     Safe Fallback Vector=<1 0 0>
    )) {isSetParameters=true}
  ]
  particleSpawnScript [
   &InitializeParticle(
    Color=&MakeLinearColorFromVectorAndFloat(
     "Vector (RGB)"=<1 1 1>
     "Float (Alpha)"=&Multiply_Float(A=&RandomRangeFloat(Minimum=0.75, Maximum=1), B=0.5)
    )
    Lifetime=0.5
    Mesh Scale Mode=::NewEnumerator3
    Mesh Uniform Scale=&RandomRangeFloat(Minimum=2.333, Maximum=2.999)
    Position Offset=<0 0 222>
    Sprite Size Mode=::NewEnumerator3
    Uniform Sprite Size=999
    UsePositionOffset=true
    Write Material Random=true
   )
   &InitialMeshOrientation(
    Mesh Axis To Orient=<0 0 -1>
    Mesh Orientation Mode=::NewEnumerator2
    Orientation Coordinate Space=::World
    Orientation Vector=$Emitter.OrientationVector
    Use Rotation Vector=false
   )
  ]
  particleUpdateScript [
   &ParticleState()
   &ScaleColor(
    Scale Alpha=&Multiply_Float(
     A=&FloatFromCurve(
      FloatCurve=DataInterface() {propertyValues {Curve {keys [
          {interpMode=RCIM_Cubic, leaveTangent=3.277, tangentMode=RCTM_Break, time=0, value=0}
          {interpMode=RCIM_Cubic, tangentMode=RCTM_User, time=0.5, value=1}
          {arriveTangent=-3.7507, tangentMode=RCTM_User, time=1, value=0}
         ]}}}
      Scale Curve=$System.DeployFadeIn
     )
     B=&Power(a=$System.Timeline.Time, b=3)
    )
    ScaleA=true
    ScaleRGB=true
   )
  ]
  rendererClasses [&NiagaraMeshRendererProperties]
  renderers [{rendererClass=&NiagaraMeshRendererProperties, rendererIndex=0}]
 }
}
summary {
 systemName=NS_Glider_OliveStomp_Trail
 emitters [
  {emitterName=Smokes, rendererClasses [&NiagaraMeshRendererProperties]}
  {emitterName=Flames, rendererClasses [&NiagaraSpriteRendererProperties]}
  {emitterName=LightAndGlow, rendererClasses [&NiagaraMeshRendererProperties, &NiagaraSpriteRendererProperties]}
  {emitterName=Wind, rendererClasses [&NiagaraMeshRendererProperties]}
 ]
 userVariables [
  User.bIsFrontEnd {type=NiagaraBool}
  User.bIsFrontEndPreview {type=NiagaraBool}
  User.bIsFullyDeployed {type=NiagaraBool}
  User.DeployDelay {type=NiagaraFloat, defaultValue=1.25, description=""}
  User.DeployDuration {type=NiagaraFloat, defaultValue=2, description=""}
  User.ForwardDot {type=NiagaraFloat, defaultValue=0, description=""}
  User.RightDot {type=NiagaraFloat, defaultValue=0, description=""}
 ]
}
